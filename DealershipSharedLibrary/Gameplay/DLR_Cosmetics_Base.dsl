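; Base Blueprint for dealership cosmetic actors that resolves a paint color for each placed instance.
; The construction script runs Update Painted Color, so the result updates when the actor is constructed.
; That event looks up the paint variant in a color paint data table through an FX function library helper.
; A custom color flag overrides the looked-up color and forces the painted flag on; begin play, overlap and tick events are empty.
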(event EventBeginPlay)

(event Collision|EventActorBeginOverlap
  (params (OtherActor :type Actor%20Object%20Reference)))

(event EventTick
  (params (DeltaSeconds :type Float%20%28single-precision%29)))

(event Custom|Update%20Painted%20Color
  (bind (_paintedcolor _bispainted) (Class|DLRFunctionLibraryFX|DLRSelectPaintedVariant (Math|Conversions|ToInteger(Byte) (Variables|ColorVariant|GetPaint%20Variant)) (Variables|ColorVariant|GetColor%20Paint%20DT)))
  (Variables|ColorVariant|SetPainted%20Color (Math|Color|SelectColor _paintedcolor (Variables|ColorVariant|GetCustomColor) (Variables|ColorVariant|GetUseCustomColor)))
  (Variables|ColorVariant|SetbIsPainted (select (Variables|ColorVariant|GetUseCustomColor) true _bispainted)))

(fn ConstructionScript
  (CallFunction|UpdatePaintedColor))
