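; Blueprint function library for dealership FX that resolves a paint color ID to a color variant from a data table.
; DLR_GetVariantRowFromPaintID maps the one-based ID to a row name by position, clamping at the first row and wrapping by modulo past the last.
; DLR_SelectPaintedVariant treats zero as unpainted, returning either the row's paint color or a default color.
; It also returns the painted flag and whether a painted row lookup succeeded.
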(fn DLR_GetVariantRowFromPaintID
  (params (PaintColorID :type Integer) (ColorVariantDT :type Data%20Table%20Object%20Reference%20%28by%20ref%29))
  (bind _outrownames (DataTable|GetDataTableRowNames (Variables|GetColorVariantDT)))
  (return (Utilities|Array|Get(aref) _outrownames (Math|Integer|%25%28Integer%29 (Math|Integer|Max(Integer) (- (Variables|GetPaintColorID) 1)) (Utilities|Array|Length _outrownames)))))

(fn DLR_SelectPaintedVariant
  (params (PaintColorID :type Integer) (ColorVariantDT :type Data%20Table%20Object%20Reference%20%28by%20ref%29))
  (LocalVariables|SetIsPainted (Math|Conversions|ToBoolean(Integer) (Variables|GetPaintColorID)))
  (bind _row_name (CallFunction|DLRGetVariantRowfromPaintID (Variables|GetPaintColorID) (Variables|GetColorVariantDT)))
  (bind _returnvalue (Utilities|GetDataTableRow (Variables|GetColorVariantDT) _row_name)
    (:then
      (LocalVariables|SetSuccess true))
    (:RowNotFound
      (LocalVariables|SetSuccess false)))
  (return (select (LocalVariables|GetIsPainted) (Utilities|Struct|BreakStructDLRFXPaintColors _returnvalue) (LocalVariables|GetDefault%20Color)) (LocalVariables|GetIsPainted) (and (LocalVariables|GetIsPainted) (LocalVariables|GetSuccess))))
